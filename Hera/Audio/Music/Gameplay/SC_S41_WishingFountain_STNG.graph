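# Sound cue for the Wishing Fountain music stinger (S41) in the Hera gameplay music set.
# A single wave player feeds the stinger wave straight to the output, with no randomization, mixing or modulation.
# It plays through the Fortnite gameplay music sound class and reuses the diegetic Squid Wharf restaurant attenuation asset for positional falloff.
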SoundCue /Hera/Audio/Music/Gameplay/SC_S41_WishingFountain_STNG
root=WavePlayer_0
nodes [{id=WavePlayer_0, properties {soundWaveAssetPtr=&/MX_S41_WishingFountain_STNG}}]
properties {
 attenuationSettings=/Hera/Audio/Music/Diegetic/SquidWharf/ATT_SquidWharf_Restaurant
 maxDistance=2000
 soundClassObject=/Game/Sounds/FortSoundClassesAndModes/Fort_Music_Gameplay
}
